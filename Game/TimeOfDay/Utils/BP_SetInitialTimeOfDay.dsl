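; Blueprint that sets the level's initial time of day through the Fortnite Day Sequence time-of-day manager.
; An async lookup at BeginPlay caches the manager and binds the Day Sequence modifier to it.
; Zero speed freezes time through the modifier's FixedTime cycle; otherwise speed becomes 1 and the manager's time is set.
; In editor, the construction script previews the time, and setup re-runs when the subsystem's Day Sequence actor changes.
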(event EventBeginPlay
  (CallFunction|SetupTODM))

(event Custom|Setup%20TODM
  (bind _returnvalue (WorldSubsystems|GetDaySequenceSubsystem))
  (Utilities|FlowControl|Sequence
    (:then_0
      (Variables|TimeOfDay|GetAsyncWaitForTODMTask
        (:then)
        (:else
          (bind (_asynctask _todm) (Default|GetTODM self)
            (:OnAsyncTaskActivated
              (Variables|TimeOfDay|SetAsyncWaitForTODMTask _asynctask))
            (:Out
              (bind _asfort_day_sequence_actor (Utilities|Casting|CastToFortDaySequenceActor _todm)
                (:then
                  (bind _output_get (Variables|TimeOfDay|SetCachedTODM _asfort_day_sequence_actor))
                  (DaySequence|BindtoDaySequenceActor (Variables|Default|GetDaySequenceModifier) _output_get :!class /Script/DaySequence.DaySequenceModifierComponent)
                  (CallFunction|PostTODMCached))
                (:CastFailed
                  (Fortnite|FortShippingLog self "Global Time of Day Manager is not a Day Sequence Actor!" true))))))) ; This will always be invalid in editor as we don't rely on the async task
    )
    (:then_1
      ;[1 1 1] Listen for TODM changes while in editor (we no longer rely on async actions in editor, only in game)
      (if (Editor|InEditor)
        (DaySequence|UnbindEventfromOnDaySequenceActorSet _returnvalue (delegate Custom|OnDaySequenceActorSet))
        (DaySequence|AssignOnDaySequenceActorSet _returnvalue (delegate Custom|OnDaySequenceActorSet))))))

(event Custom|OnDaySequenceActorSet
  (params (NewActor :type Day%20Sequence%20Actor%20Object%20Reference))
  ;[1 1 1] Listen for TODM changes while in editor (we no longer rely on async actions in editor, only in game)
  (CallFunction|SetupTODM))

(event Custom|Post%20TODM%20Cached
  (bind _initialtimeofday (Variables|TimeOfDay|GetInitialTimeOfDay))
  (bind _daysequencemodifier (Variables|Default|GetDaySequenceModifier))
  (if (Math|Float|NearlyEqual(Float) (Variables|TimeOfDay|GetInitialTimeOfDaySpeed))
    ;[1 1 1] Static Time of Day Override (Freezes Time)
    (Utilities|FlowControl|Sequence
      (:then_1
        (Variables|/Script/DaySequence.DaySequenceModifierComponent|SetDayNightCycle "FixedTime" _daysequencemodifier)
        (Variables|/Script/DaySequence.DaySequenceModifierComponent|SetDayNightCycleTime (Variables|TimeOfDay|GetInitialTimeOfDay) _daysequencemodifier)))
    (else
      ;[1 1 1] Normal Time of Day Override (Sets Time of Day, does not touch Time of Day Speed)
      (bind _asfort_time_of_day_manager_interface (Utilities|Casting|CastToFortTimeOfDayManagerInterface (Variables|TimeOfDay|GetCachedTODM)))
      (Utilities|FlowControl|Sequence
        (:then_0
          (Variables|TimeOfDay|SetInitialTimeOfDaySpeed 1.0) ; We only support speed factors of 0 and 1 at the moment
        )
        (:then_1
          (TimeOfDay|SetTimeOfDay _asfort_time_of_day_manager_interface _initialtimeofday :!class /Script/FortniteGame.FortTimeOfDayManagerInterface)
          (Fortnite|FortShippingLog self (Utilities|String|Append "Init TOD = " (Utilities|String|ToString(Float) _initialtimeofday))))))))

(event Custom|Setup%20TODM%20In%20Editor
  (bind _initialtimeofday (Variables|TimeOfDay|GetInitialTimeOfDay))
  (if (Editor|InEditor)
    (bind (_asynctask _todm) (Default|GetTODM 0)
      (:Out
        (bind _asfort_day_sequence_actor (Utilities|Casting|CastToFortDaySequenceActor _todm))
        (TimeOfDay|SetTimeOfDayPreview _asfort_day_sequence_actor _initialtimeofday)
        (Fortnite|FortShippingLog self (Utilities|String|Append "SetTimeOfDayPreview = " (Utilities|String|ToString(Float) _initialtimeofday)))))))

(fn ConstructionScript
  (CallFunction|SetupTODMinEditor))
